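# M_3D_Star is an unlit translucent effects material for mesh particles, including Niagara, stored among the Husk effects.
# The green channel of a flare burst texture serves as a mask, tinted by particle color and passed into the MF_FX_EmissiveAdjustments function.
# The same channel is multiplied by particle alpha, so particle systems control both tint and fade.
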Material /Game/Effects/Fort_Effects/Materials/Husk/M_3D_Star
graph {
 properties {
  bUsedWithMeshParticles=true
  bUsedWithNiagaraMeshParticles=true
  blendMode=BLEND_Translucent
  pixelDepthOffsetMode=PDOM_Legacy
  shadingModel=MSM_Unlit
  usedShadingModels=Unlit
 }
 nodes [
  n1 = TextureSample() {texture=/Game/Effects/Fort_Effects/Materials/T_EFX_Flare_Burst_01}
  n2 = ParticleColor()
  MaterialFunctionCall(Emissive=Multiply(A=@n1.G, B=@n2.RGB)) {materialFunction=&MF_FX_EmissiveAdjustments}
  Multiply(A=@n2.A, B=@n1.G)
 ]
}
